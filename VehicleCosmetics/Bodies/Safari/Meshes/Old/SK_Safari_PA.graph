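# Physics asset for an older Safari vehicle body skeletal mesh in the vehicle cosmetics content, stored under an Old folder.
# It holds a single physics body on the root bone, with no collision shapes listed.
# It has no constraints, so it defines no articulated or jointed parts.
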PhysicsAsset /VehicleCosmetics/Bodies/Safari/Meshes/Old/SK_Safari_PA
bodies {root []}
constraints []
